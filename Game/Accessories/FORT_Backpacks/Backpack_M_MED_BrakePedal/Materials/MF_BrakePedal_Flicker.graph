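# Material function for the BrakePedal backpack that renders a flickering, glitching holographic screen from a texture parameter.
# Switchable glitches distort the texture UVs with a grid effect, scrolling electricity noise and sine wiggle, with a sine-and-random-flicker mask defining intermittent bursts.
# The squared, brightened sample gets scanlines and a random blink dim.
# The result overrides Emissive in the material attributes alongside scaled Specular, and is also output separately; the Opacity output is unconnected.
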MaterialFunction /Game/Accessories/FORT_Backpacks/Backpack_M_MED_BrakePedal/Materials/MF_BrakePedal_Flicker
graph {nodes [
  n1 = GetMaterialAttributes(None=NamedRerouteUsage()) {attributeGetTypes [#1]}
  n2 = FunctionInput(Preview=TextureCoordinate()) {bUsePreviewValueAsDefault=true, inputName=TextureUV, inputType=FunctionInput_Vector2, sortPriority=2}
  n3 = LinearInterpolate(
   A=@n2
   B=Saturate(None=MaterialFunctionCall!(UV=@n2) {materialFunction=/Game/Global/Materials/MaterialGym/Materials/HolographicTexture/MF_HolographicTextureGlitch_Grid}.Glitched UV)
   Alpha=ScalarParameter() {defaultValue=0.5, group=HolographicTexture_Glitch, parameterName=GlitchGridIntensity}
  )
  n4 = NamedRerouteUsage()
  n5 = Rotator(Coordinate=TextureCoordinate(), Time=ScalarParameter() {group=HolographicTexture_Glitch, parameterName=GlitchUVRotation})
  n6 = ComponentMask(None=VectorParameter() {defaultValue=<rgba 0 10 0 1>}.RGB) {g=true, r=true}
  n7 = Constant() {r=0.5}
  n8 = ScalarParameter() {defaultValue=5, group=HolographicTexture_Glitch, parameterName=StretchedScrollingSpeed}
  n9 = TextureSample(UVs=Panner(Coordinate=Multiply(A=@n5, B=@n6), Speed=AppendVector(A=@n7, B=@n8)) {speedX=0.5, speedY=5}) {samplerType=SAMPLERTYPE_Masks, texture=/Game/Characters/Player/Male/Medium/Bodies/M_MED_Storm_Soldier/Textures/ElectricityNoise_M}
  n10 = Multiply(
   A=@n9.R
   B=LinearInterpolate(B=ScalarParameter() {defaultValue=0.1, group=HolographicTexture_Glitch, parameterName=StretchedGlitchIntensity}, Alpha=NamedRerouteUsage()) {constB=0.1}
  ) {constB=0.2}
  n11 = StaticSwitchParameter(True=Add(A=LinearInterpolate(A=@n2, B=@n3, Alpha=@n4), B=@n10), False=Add(A=@n2, B=@n10)) {defaultValue=true, group=HolographicTexture_Glitch, parameterName=UseGridGlitch, sortPriority=0}
  n12 = Multiply(A=Time()) {constB=10}
  n13 = Multiply(
   A=NamedRerouteUsage()
   B=Multiply(
    A=Sine(None=@n12)
    B=ScalarParameter() {defaultValue=0.0075, desc=Wiggle Intensity, group=HolographicTexture_Glitch, parameterName=GlitchWiggle}
   ) {constB=0.007}
  )
  n14 = TextureSampleParameter2D(UVs=StaticSwitchParameter(True=Add(A=@n11, B=@n13), False=@n2) {defaultValue=true, group=HolographicTexture_Glitch, parameterName=UseGlitch, sortPriority=-1}) {
   group=HolographicTexture
   parameterName=HologramTexture
   samplerSource=SSM_Clamp_WorldGroupSettings
   sortPriority=-20
   texture=/Game/2dAssets/Sprays/Season25/Textures/T-Sprays-Season25-S25-GalaxyCupReward
  }
  n15 = Multiply(
   A=Multiply(A=@n14.RGB, B=@n14.RGB)
   B=ScalarParameter() {defaultValue=1, group=HolographicTexture, parameterName=TextureBrightness, sortPriority=-19}
  )
  n16 = ScalarParameter() {defaultValue=1.5, group=HolographicTexture, parameterName=Scanline_ColorMultiplier, sortPriority=0}
  n17 = VectorParameter() {group=HolographicTexture, parameterName=Scanline_ColorOvelay, defaultValue=<rgba 0.6094 0.6094 0.6094 1>}
  n18 = FunctionInput(Preview=TextureCoordinate()) {bUsePreviewValueAsDefault=true, inputName=Scanlines_UV, inputType=FunctionInput_Vector2, sortPriority=2}
  n19 = LinearInterpolate(
   A=@n15
   B=Add(A=Multiply(A=@n15, B=@n16), B=@n17.RGB)
   Alpha=SmoothStep(
    Min=ScalarParameter() {group=HolographicTexture, parameterName=Scanline_Thickness, sortPriority=0}
    Value=MaterialFunctionCall!(ScanLine_Tiling_1, Scanline_PannerSpeed_1, ScanLine_Tiling, Scanline_PannerSpeed, Scanline_Rotation, UVs=@n18) {desc=ScanLine, materialFunction=/Game/Global/Materials/MaterialGym/Materials/HolographicTexture/MF_HolographicTexture_ScanLines}.Result
   ) {constMin=1}
  )
  n20 = LinearInterpolate(
   A=@n19
   B=Multiply(A=@n19, B=ScalarParameter() {defaultValue=0.9, parameterName=ScreenBlinkIntensity})
   Alpha=MaterialFunctionCall() {materialFunction=/Game/MF__FloatFlickerRandom_Generic}.RandomFlickerFloat
  )
  n21 = ScalarParameter() {defaultValue=1, parameterName=SpecularMultiplier}
  FunctionOutput(None=SetMaterialAttributes!(MaterialAttributes=@n1.MaterialAttributes, EmissiveColor=@n20, Specular=Multiply(A=@n1.Specular, B=@n21)) {attributeSetTypes [#2, #1]}) {outputName=MaterialAttributes, sortPriority=-1}
  FunctionOutput(None=@n20) {outputName=EmissiveColor, sortPriority=1}
  FunctionOutput() {outputName=Opacity}
  Multiply(A=@n19, B=@n15)
  n22 = Constant() {r=1}
  n23 = Subtract(A=@n22, B=ScalarParameter() {defaultValue=0.075, group=HolographicTexture_Glitch, parameterName=GlitchDuration, sortPriority=10})
  n24 = ScalarParameter() {defaultValue=0.2, group=HolographicTexture_Glitch, parameterName=GlitchFrequency, sortPriority=10}
  n25 = Sine(None=Multiply(A=Time(), B=@n24) {constB=0.2})
  n26 = Sine(None=Multiply(A=Time() {period=1}, B=Multiply(A=@n24) {constB=2}) {constB=0.2}) {period=1.1}
  n27 = Add(A=Multiply(A=LinearInterpolate(A=@n25, Alpha=@n26)) {constB=0.5}) {constB=0.5}
  n28 = MaterialFunctionCall() {materialFunction=&MF__FloatFlickerRandom_Generic}
  NamedRerouteDeclaration(None=Multiply(
    A=SmoothStep(Min=@n23, Max=@n22, Value=@n27) {constMax=0.9, constMin=0.8}
    B=NamedRerouteDeclaration(None=@n28.RandomFlickerFloat) {name=RandomFlicker, nodeColor=<rgba 1 0.3047 0 1>}
   )) {name=GlitchFrequencyMask, nodeColor=<rgba 0.9377 1 0.8953 1>}
  NamedRerouteDeclaration(None=@n14.RGBA) {name=Texture RGBA, nodeColor=<rgba 1 0 0.8906 1>}
  n29 = GetMaterialAttributes(None=FunctionInput() {inputName=MaterialAttributes, inputType=FunctionInput_MaterialAttributes, sortPriority=-1}) {attributeGetTypes [
    #3
    #4
    #2
    #5
   ]}
  NamedRerouteDeclaration(None=@n29.BaseColor) {name=OG Base Color, nodeColor=<rgba 0 0.7422 1 1>}
  NamedRerouteDeclaration(None=@n29.WorldPositionOffset) {name=OG WPO, nodeColor=<rgba 0 1 0.6016 1>}
  NamedRerouteDeclaration(None=@n29.EmissiveColor) {name=OG Emissive, nodeColor=<rgba 0 1 0.1797 1>}
  NamedRerouteDeclaration(None=@n29.Opacity) {name=OG Opacity, nodeColor=<rgba 0.3422 0.8799 1 1>}
  NamedRerouteDeclaration(None=@n29.MaterialAttributes) {name=OG Material Attributes, nodeColor=<rgba 0.8984 0 1 1>}
  Subtract(A=@n9.R) {constB=0.5}
 ]}
