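# Parent character material for the BrakePedal backpack, adding a glitch flicker effect between the 2023 character PreFX and PostFX functions.
# A channel-selectable GlitchMask blends MF_BrakePedal_Flicker into the material attributes.
# A time-driven sine and a small UV sphere mask define a blinking light spot.
# A rerouted mask then interpolates the blended emissive between dimmed and full strength.
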Material /Game/Accessories/FORT_Backpacks/Backpack_M_MED_BrakePedal/Materials/M_BrakePedal_Backpack_Parent
graph {
 properties {
  bUseMaterialAttributes=true
  bUsedWithMeshDeformer=true
  bUsedWithNanite=true
  bUsedWithSkeletalMesh=true
  pixelDepthOffsetMode=PDOM_Legacy
  shadingModel=MSM_FromMaterialExpression
  usedShadingModels=Default Lit | Subsurface | Clear Coat | Subsurface Profile | Cloth
 }
 nodes [
  n1 = Multiply(A=Time())
  n2 = Saturate(None=Add(A=Sine(None=@n1)))
  n3 = NamedRerouteDeclaration(None=SphereMask(A=TextureCoordinate(), B=Constant2Vector() {g=0.37, r=0.42}) {attenuationRadius=0.0175}) {name=LightMask, nodeColor=<rgba 1 0 0.3516 1>}
  LinearInterpolate(
   A=NamedRerouteDeclaration(None=Multiply(A=@n2, B=@n3)) {name=LightBlink, nodeColor=<rgba 0 1 0.8125 1>}
   B=TextureSample() {texture=/Game/Accessories/FORT_Backpacks/Backpack_M_MED_BrakePedal/Textures/T_Backpack_BrakePedal_E}.RGB
  ) {constAlpha=0.75}
  n4 = MaterialFunctionCall() {materialFunction=&MF_CharacterParent_2023_PreFX}
  MaterialFunctionCall(Material=NamedRerouteUsage(), M=@n4.M, Pre Skinned Local Position=@n4.M, Distance Blend=@n4.M, Fresnel=@n4.M) {materialFunction=&MF_CharacterParent_2023_PostFX}
  NamedRerouteDeclaration(None=@n4.Tech Art Mask) {name=FXMask, nodeColor=<rgba 0.6406 0 1 1>}
  NamedRerouteDeclaration(None=@n4.Result) {name=inMaterialAttributes, nodeColor=<rgba 1 0.0234 0 1>}
  n5 = GetMaterialAttributes(None=NamedRerouteUsage()) {attributeGetTypes [#1, #2]}
  n6 = NamedRerouteUsage()
  n7 = GetMaterialAttributes(None=BlendMaterialAttributes(
    A=@n5.MaterialAttributes
    B=MaterialFunctionCall!(MaterialAttributes=@n5.MaterialAttributes, Scanlines_UV, TextureUV) {materialFunction=&/MF_BrakePedal_Flicker}.MaterialAttributes
    Alpha=ChannelMaskParameter(None=@n6) {parameterName=GlitchMask}.R
   )) {attributeGetTypes [#2]}
  n8 = Multiply(A=@n7.Emissive Color) {constB=0.3}
  n9 = OneMinus(None=NamedRerouteUsage())
  NamedRerouteDeclaration(None=SetMaterialAttributes(MaterialAttributes=@n7.MaterialAttributes, Emissive Color=LinearInterpolate(A=@n8, B=@n7.Emissive Color, Alpha=@n9)) {attributeSetTypes [#2]}) {name=outMaterialAttributes, nodeColor=<rgba 0.5938 0 1 1>}
 ]
}
